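# Translucent unlit refraction material for Niagara mesh particles, apparently the twisting air distortion in the supply drop air-burst spawn effect.
# A tiled, panned normal map, with Y tiling scaled by a particle dynamic parameter, supplies the ripple pattern.
# UV-based edge masks, optional depth and camera fades, and particle alpha blend refraction from neutral toward a per-particle strength, gated by Use Particle Refraction.
# The near/far refraction blend reads the same strength on both sides.
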Material /SupplyDropRadioGameplay/FX/AirDrop_AirBurst_Spawn/Materials/M_FX_SDR_TwistDistort
graph {
 properties {
  bUsedWithNiagaraMeshParticles=true
  blendMode=BLEND_Translucent
  pixelDepthOffsetMode=PDOM_Legacy
  refractionMethod=RM_IndexOfRefraction
  shadingModel=MSM_Unlit
  usedShadingModels=Unlit
 }
 nodes [
  Constant()
  n1 = DynamicParameter() {paramNames [RippleTiling-Multiply, Refraction Strength, Refraction Strength Far, Param4]}
  n2 = MaterialFunctionCall(Distance=ScalarParameter() {defaultValue=100, group=Camera Fade, parameterName=Camera Fade Distance}) {materialFunction=&Near_Camera_Fade}
  n3 = TextureCoordinate()
  n4 = Multiply(
   A=OneMinus(None=ComponentMask(None=@n3) {g=true})
   B=ScalarParameter() {defaultValue=3, group=EdgeMask, parameterName=InsideEdge-Mask_Hardness}
  )
  n5 = Multiply(
   A=ComponentMask(None=TextureCoordinate()) {g=true}
   B=ScalarParameter() {defaultValue=3, group=EdgeMask, parameterName=OutsideEdge-Mask_Hardness}
  )
  n6 = Multiply(
   A=Power(Base=Clamp(None=@n5), Exp=ScalarParameter() {defaultValue=2, group=EdgeMask, parameterName=OutsideEdge-Mask_Exponent})
   B=Power(Base=Clamp(None=@n4), Exp=ScalarParameter() {defaultValue=2, group=EdgeMask, parameterName=InsideEdge-Mask_Exponent})
  )
  n7 = LinearInterpolate(
   A=@n6
   B=DepthFade(Opacity=@n6, FadeDistance=ScalarParameter() {defaultValue=100, group=Depth Fade, parameterName=Depth Fade Distance})
   Alpha=ScalarParameter() {group=Depth Fade, parameterName=Use Depth Fade}
  ) {constA=1}
  n8 = LinearInterpolate(A=@n7, B=Multiply(A=@n7, B=@n2.Result), Alpha=ScalarParameter() {group=Camera Fade, parameterName=Use Camera Fade}) {constA=1}
  n9 = ParticleColor()
  LinearInterpolate(
   A=Constant() {r=1}
   B=LinearInterpolate(A=@n1.Refraction Strength, B=@n1.Refraction Strength, Alpha=@n2.Result)
   Alpha=Multiply(A=ScalarParameter() {parameterName=Use Particle Refraction}, B=Multiply(A=@n8, B=@n9.A))
  )
  n10 = ScalarParameter() {defaultValue=1, parameterName=Y-Tile}
  n11 = DynamicParameter() {paramNames [RippleTiling-Multiply, Refraction Strength, Refraction Strength Far, Param4]}
  n12 = Multiply(
   A=TextureCoordinate()
   B=AppendVector(A=ScalarParameter() {defaultValue=1, parameterName=X-Tile}, B=Multiply(A=@n10, B=@n11.RippleTiling-Multiply))
  )
  n13 = AppendVector(A=ScalarParameter() {parameterName=X-Offset}, B=ScalarParameter() {parameterName=Y-Offset})
  n14 = ScalarParameter() {parameterName=X-PanSpeed}
  n15 = ScalarParameter() {parameterName=Y-PanSpeed}
  n16 = TextureSampleParameter2D(UVs=Panner(Coordinate=Add(A=@n12, B=@n13), Speed=AppendVector(A=@n14, B=@n15))) {
   parameterName=NormalMap
   samplerType=SAMPLERTYPE_Normal
   texture=/SupplyDropRadioGameplay/FX/AirDrop_AirBurst_Spawn/Textures/T_FX_SDR_AlternatingNormals
  }
  n17 = Transform(None=@n16.RGB) {transformType=TRANSFORM_Camera}
  MaterialFunctionCall(Emissive=@n17) {materialFunction=&MF_FX_EmissiveAdjustments}
  Multiply()
  Reroute(None=@n17)
  Reroute(None=@n17)
  ScalarParameter() {defaultValue=1.1, parameterName=RefractionStrengthFar}
 ]
}
