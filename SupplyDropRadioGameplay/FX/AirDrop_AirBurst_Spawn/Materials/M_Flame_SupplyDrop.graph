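# Additive, unlit, two-sided flame material for the supply drop air-burst effect, usable on Niagara sprites and meshes.
# Panned noise layers and a V gradient build a contrast-shaped mask that blends two colors and, with gradient falloff, drives opacity, depth-faded off mobile.
# WPO combines directional sine sway with noise and band displacement along normals, optionally blending in a gradient bulge driven by dynamic parameters.
# Vertex red varies speed and UVs; a static switch applies particle color.
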Material /SupplyDropRadioGameplay/FX/AirDrop_AirBurst_Spawn/Materials/M_Flame_SupplyDrop
graph {
 properties {
  bUsedWithMeshDeformer=true
  bUsedWithNiagaraMeshParticles=true
  bUsedWithNiagaraSprites=true
  bUsedWithSkeletalMesh=true
  blendMode=BLEND_Additive
  pixelDepthOffsetMode=PDOM_Legacy
  shadingModel=MSM_Unlit
  twoSided=true
  usedShadingModels=Unlit
 }
 nodes [
  n1 = Constant()
  n2 = Time()
  n3 = VertexColor()
  n4 = LinearInterpolate(Alpha=@n3.R) {constB=0.5}
  n5 = Sine(None=Add(A=Multiply(A=@n2, B=@n4), B=LinearInterpolate(Alpha=@n3.R) {constB=2}))
  n6 = AppendVector(A=@n5, B=@n5)
  n7 = Transform(None=Multiply(A=AppendVector(A=@n1, B=@n6), B=VectorParameter() {group=WPO, parameterName=FlameDirection, defaultValue=<rgba 0 -1 1 1>}.RGB) {constA=1}) {transformSourceType=TRANSFORMSOURCE_Local}
  n8 = ComponentMask(None=TextureCoordinate()) {g=true}
  n9 = LinearInterpolate(
   B=ScalarParameter() {defaultValue=9, group=WPO, parameterName=Distort}
   Alpha=Power(Base=OneMinus(None=@n8), Exp=ScalarParameter() {defaultValue=3, group=WPO, parameterName=WPOContrast})
  )
  n10 = OneMinus(None=ComponentMask(None=TextureCoordinate()) {g=true})
  n11 = LinearInterpolate(B=ScalarParameter() {defaultValue=8, group=WPO, parameterName=Distort Edge}, Alpha=Multiply(A=@n10, B=@n10))
  n12 = ComponentMask(None=VectorParameter() {group=FlameShape, parameterName=Flames_UVScale, defaultValue=<rgba 0.6 0.4 0 0>}.RGB) {g=true, r=true}
  n13 = Add(A=TextureCoordinate() {uTiling=4}, B=VertexColor().R)
  n14 = ScalarParameter() {group=FlameShape, parameterName=Flame_UVOffset}
  n15 = LinearInterpolate(Alpha=VertexColor().R) {constA=0.5, constB=2}
  n16 = Constant() {r=1}
  n17 = Multiply(
   A=Time()
   B=Multiply(
    A=ScalarParameter() {defaultValue=2, group=FlameShape, parameterName=Flame_Speed}
    B=StaticSwitchParameter(True=@n15, False=@n16) {defaultValue=true, parameterName=UseVertexColorCtrlFlameSpeed}
   )
  )
  n18 = Panner(Coordinate=Add(A=Multiply(A=@n12, B=@n13), B=@n14), Time=@n17) {speedX=0.2, speedY=1}
  n19 = Multiply(
   A=LinearInterpolate(Alpha=TextureSample!(UVs=@n18, Tex, Level, Apply View MipBias) {mipValueMode=TMVM_MipLevel, texture=&T_NoisePack_01}.G) {constA=-1}
   B=VertexNormalWS()
  )
  n20 = ComponentMask(None=TextureCoordinate()) {g=true}
  n21 = ScalarParameter() {defaultValue=-0.125, group=WPO, parameterName=Distort_offset}
  n22 = Power(Base=Sine(None=Add(A=@n20, B=@n21)) {period=2}, Exp=ScalarParameter() {defaultValue=6, group=WPO, parameterName=Distort_power})
  n23 = ScalarParameter() {defaultValue=1, group=WPO, parameterName=Distort_Multi}
  n24 = Add(A=Add(A=Multiply(A=@n7, B=@n9), B=Multiply(A=@n11, B=@n19)), B=Multiply(A=VertexNormalWS(), B=Multiply(A=@n22, B=@n23))) {constB=0}
  n25 = DynamicParameter() {defaultValue=<rgba 1 1.5 0 0>, paramNames [WPO Intensity, Panning, Param3, Param4]}
  n26 = TextureCoordinate()
  n27 = TextureSampleParameter2D(UVs=Clamp(None=Subtract(A=@n25.Panning, B=@n26))) {
   group=WPO
   parameterName=WPO Texture Second
   samplerSource=SSM_Clamp_WorldGroupSettings
   texture=/Game/Global/Textures/Default/Masks_and_Gradients/gradient_stasis
  }
  n28 = Multiply(
   A=ScalarParameter() {defaultValue=1, parameterName=Power Multiply}
   B=Power(Base=@n27.RGB, Exp=ScalarParameter() {defaultValue=1, parameterName=Power Exponent})
  )
  n29 = Multiply(A=ScalarParameter() {defaultValue=300, parameterName=WPO Intensity}, B=Multiply(A=VertexNormalWS(), B=Clamp(None=@n28)))
  LinearInterpolate(
   A=Multiply(A=@n24, B=ScalarParameter() {defaultValue=1, group=WPO, parameterName=WPO_OverallIntensity})
   B=Multiply(A=Add(A=@n24, B=@n29), B=@n25.WPO Intensity)
   Alpha=ScalarParameter() {group=WPO, parameterName=Use Dynamic WPO}
  ) {constA=1}
  n30 = TextureCoordinate() {vTiling=0.5}
  n31 = VertexColor()
  n32 = Add(A=Multiply(A=Add(A=@n30, B=@n31.R), B=@n12), B=@n14)
  n33 = MaterialFunctionCall(
   In=TextureSample!(UVs=Panner(Coordinate=@n32) {speedX=-0.5, speedY=1}, Tex, Apply View MipBias) {texture=&T_NoisePack_01}.G
   Contrast=Constant() {r=0.2}
  ) {materialFunction=&CheapContrast}
  n34 = TextureSample(UVs=@n18) {texture=&T_NoisePack_01}
  n35 = Add(A=TextureCoordinate() {uTiling=6, vTiling=1.5}, B=VertexColor().R)
  n36 = Panner(Coordinate=Add(A=@n14, B=Multiply(A=@n12, B=@n35)), Time=@n17) {speedX=-0.2, speedY=1}
  n37 = Add(
   A=Multiply(A=Add(A=@n33.Result, B=@n34.G), B=TextureSample(UVs=@n36) {texture=&T_NoisePack_01}.G)
   B=Power(
    Base=MaterialFunctionCall() {materialFunction=&LinearGradient}.VGradient
    Exp=ScalarParameter() {defaultValue=2, group=FlameShape, parameterName=FireLine}
   )
  )
  n38 = Clamp(None=MaterialFunctionCall(In=Power(Base=@n37) {constExponent=0.5}, Contrast=ScalarParameter() {defaultValue=0.5, group=FlameShape, parameterName=HardEdgeness}) {materialFunction=&CheapContrast}.Result)
  n39 = LinearInterpolate(
   A=VectorParameter() {group=FlameShape, parameterName=Color_A, defaultValue=<rgba 10 0 0.1768 1>}.RGB
   B=VectorParameter() {group=FlameShape, parameterName=Color_B, defaultValue=<rgba 10 1.3123 0 1>}.RGB
   Alpha=@n38
  )
  n40 = ParticleColor()
  MaterialFunctionCall(Emissive=StaticSwitchParameter(True=Multiply(A=@n39, B=@n40.RGB), False=@n39) {parameterName=Use particle color}) {materialFunction=&MF_FX_EmissiveAdjustments}
  n41 = Multiply(
   A=Power(
    Base=MaterialFunctionCall() {materialFunction=&LinearGradient}.VGradient
    Exp=ScalarParameter() {defaultValue=4, group=Alpha, parameterName=VGradientContrast_Top}
   )
   B=ScalarParameter() {defaultValue=2, group=Alpha, parameterName=Enhance}
  )
  n42 = Multiply(A=ScalarParameter() {defaultValue=2, group=Alpha, parameterName=MasterAlpha}, B=Multiply(A=@n38, B=Saturate(None=@n41)))
  n43 = DepthFade(Opacity=@n42, FadeDistance=ScalarParameter() {defaultValue=2, group=Alpha, parameterName=Fade Distance})
  n44 = ShadingPathSwitch(Default=@n43, Deferred=@n43, Mobile=@n42) {desc=Migrated from FeatureLevelSwitch}
  StaticSwitchParameter(True=Multiply(A=@n40.A, B=@n44), False=@n44) {parameterName=Use particle color}
 ]
}
