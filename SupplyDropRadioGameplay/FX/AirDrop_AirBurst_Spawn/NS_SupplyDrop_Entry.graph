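# Niagara effect for a supply drop's entry air burst, built from three mesh-particle emitters.
# Central_Flames and a raised, rotating Particle_Emitter mesh spawn as bursts, live for User.Time, shrink and fade, with width set by User.Scale.
# MainAirFlash continuously spawns short-lived, pale-blue, vertically oriented flashes that scale by curve.
# Alpha fades follow system loop age, and dynamic material parameters drive the flame and flash materials.
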NiagaraSystem /SupplyDropRadioGameplay/FX/AirDrop_AirBurst_Spawn/NS_SupplyDrop_Entry
dependencies {usedDataInterfaces [&NiagaraDataInterfaceCurve], usedDynamicInputs [], usedModules [&SystemState]}
emitters {
 Central_Flames {
  emitterUpdateScript [
   &EmitterState()
   &SpawnBurst_Instantaneous(Use Spawn Probability=false, UseLoopCountLimit=true)
  ]
  particleSpawnScript [
   &InitializeParticle(
    Lifetime=$User.Time
    Mesh Scale=&MakeVector(
     X=&Multiply_Float(A=50, B=$User.Scale)
     Y=&Multiply_Float(A=50, B=$User.Scale)
     Z=15
    )
    Mesh Scale Mode=::NewEnumerator4
    Position Offset=<0 0 25>
    UsePositionOffset=true
    Write Material Random=true
   )
   &InitialMeshOrientation(Mesh Axis To Orient=<0 0 1>, Mesh Orientation Mode=::NewEnumerator2, Orientation Vector=<0 0 1>, Use Rotation Vector=true)
  ]
  particleUpdateScript [
   &ParticleState()
   &ScaleMeshSize(Scale Factor=&MakeVector(X=1, Y=1, Z=3))
   &ScaleMeshSize(Scale Factor=&VectorFromFloat(Value=&FloatFromCurve(
      FloatCurve=DataInterface() {propertyValues {Curve {keys [
          {interpMode=RCIM_Cubic, tangentMode=RCTM_User, time=0.1, value=1}
          {arriveTangent=-0.562, interpMode=RCIM_Cubic, leaveTangent=-0.562, tangentMode=RCTM_User, time=0.3, value=0.8}
          {arriveTangent=-1.5323, interpMode=RCIM_Cubic, leaveTangent=-1.5323, tangentMode=RCTM_User, time=1, value=0.5}
         ]}}}
      Scale Curve=0.2
     ))) {module=ScaleMeshSize001}
   &ScaleColor(
    Scale Alpha=&FloatFromCurve(
     FloatCurve=DataInterface() {propertyValues {Curve {keys [
         {interpMode=RCIM_Cubic, tangentMode=RCTM_User, time=0.1, value=1}
         {arriveTangent=-0.562, interpMode=RCIM_Cubic, leaveTangent=-0.562, tangentMode=RCTM_User, time=0.4, value=0.3}
         {arriveTangent=-1.7459, interpMode=RCIM_Cubic, leaveTangent=-1.7459, tangentMode=RCTM_User, time=1, value=0}
        ]}}}
     CurveIndex=$System.NormalizedLoopAge
     Scale Curve=0.1
    )
    ScaleA=true
    ScaleRGB=false
   )
   &DynamicMaterialParameters(
    Index 0 Param 1=&Lerp_Float(A=1.5, B=0.5, Alpha=$User.Time)
    Index 0 Param 2=&FloatFromCurve(FloatCurve=DataInterface() {propertyValues {Curve {keys [<at 0 2>,<at 1 0>]}}})
    Module.Param0WriteEnabled=true
    Module.Param1WriteEnabled=true
    Module.Param2WriteEnabled=true
    Module.Param3WriteEnabled=true
   )
  ]
  rendererClasses [&NiagaraMeshRendererProperties]
  renderers [{rendererClass=&NiagaraMeshRendererProperties, rendererIndex=0}]
 }
 MainAirFlash {
  emitterUpdateScript [
   &EmitterState()
   &SpawnRate(SpawnRate=10, Use Spawn Probability=false)
  ]
  particleSpawnScript [
   &InitializeParticle(
    Color=<rgba 0.4158 0.7859 1 0.1>
    Lifetime Max=0.3
    Lifetime Min=0.2
    Lifetime Mode=::NewEnumerator1
    Mesh Scale Max=<1280 320 320>
    Mesh Scale Min=<640 320 320>
    Mesh Scale Mode=::NewEnumerator2
    UsePositionOffset=false
    Write Material Random=true
   )
   &InitialMeshOrientation(Mesh Axis To Orient=<0 0 -1>, Mesh Orientation Mode=::NewEnumerator2, Use Rotation Vector=true)
  ]
  particleUpdateScript [
   &ParticleState()
   &ScaleColor(
    Scale Alpha=&FloatFromCurve(
     FloatCurve=DataInterface() {propertyValues {Curve {keys [
         {interpMode=RCIM_Cubic, tangentMode=RCTM_User, time=0.1, value=1}
         {arriveTangent=-2.7971, interpMode=RCIM_Cubic, leaveTangent=-2.7971, tangentMode=RCTM_User, time=1, value=0}
        ]}}}
     CurveIndex=$System.NormalizedLoopAge
    )
    ScaleA=true
    ScaleRGB=false
   )
   &ScaleMeshSize(Scale Factor=/Niagara/DynamicInputs/CascadeConversion/SplitVector(
     Double Component Vector Channels:ECascadeNiagaraTwoVectorChannels=::NewEnumerator1
     Double Component:NiagaraFloat=&FloatFromCurve(FloatCurve=DataInterface() {propertyValues {Curve {keys [
          {arriveTangent=6.3707, interpMode=RCIM_Cubic, time=0, value=0}
          {arriveTangent=0.5, interpMode=RCIM_Cubic, leaveTangent=0.5, time=0.2004, value=0.9893}
          <at 1 0.5 RCIM_Cubic>
         ]}}})
     Single Component:NiagaraFloat=&FloatFromCurve(FloatCurve=DataInterface() {propertyValues {Curve {keys [
          {arriveTangent=3.8376, interpMode=RCIM_Cubic, time=0, value=0}
          {arriveTangent=1, interpMode=RCIM_Cubic, leaveTangent=1, time=0.2004, value=0.5847}
          <at 1 1 RCIM_Cubic>
         ]}}})
    ))
   &ScaleMeshSize(Scale Factor=&VectorFromFloat(Value=0.075)) {module=ScaleMeshSize001}
   &DynamicMaterialParameters(
    Index 0 Param 1=1
    Index 0 Param 2=5
    Index 0 Param 3=2
    Index 0 Param 4=1
    Module.Param0WriteEnabled=true
    Module.Param1WriteEnabled=true
    Module.Param2WriteEnabled=true
    Module.Param3WriteEnabled=true
   )
  ]
  rendererClasses [&NiagaraMeshRendererProperties]
  renderers [{rendererClass=&NiagaraMeshRendererProperties, rendererIndex=0}]
 }
 Particle_Emitter {
  emitterUpdateScript [
   &EmitterState()
   &SpawnBurst_Instantaneous(Use Spawn Probability=false, UseLoopCountLimit=true)
  ]
  particleSpawnScript [
   &InitializeParticle(
    Color=<rgba 1 1 1 0.1>
    Lifetime=$User.Time
    Mesh Scale=&MakeVector(
     X=120
     Y=&Multiply_Float(A=80, B=$User.Scale)
     Z=&Multiply_Float(A=80, B=$User.Scale)
    )
    Mesh Scale Mode=::NewEnumerator4
    Position Offset=<0 0 550>
    UsePositionOffset=true
    Write Material Random=true
   )
   &InitialMeshOrientation(Mesh Axis To Orient=<0 0 1>, Mesh Orientation Mode=::NewEnumerator2, Use Rotation Vector=true)
  ]
  particleUpdateScript [
   &ParticleState()
   /SetVariables(Particles.YZScale:NiagaraFloat=&FloatFromCurve(
     FloatCurve=DataInterface() {propertyValues {Curve {keys [
         {interpMode=RCIM_Cubic, tangentMode=RCTM_User, time=0.1, value=1}
         {arriveTangent=-0.562, interpMode=RCIM_Cubic, leaveTangent=-0.562, tangentMode=RCTM_User, time=0.3, value=0.9}
         {arriveTangent=-1.349, interpMode=RCIM_Cubic, leaveTangent=-1.349, tangentMode=RCTM_User, time=1, value=0.6}
        ]}}}
     CurveIndex=$System.NormalizedLoopAge
    )) {isSetParameters=true}
   &ScaleMeshSize(Scale Factor=&MakeVector(X=1, Y=$Particles.YZScale, Z=$Particles.YZScale)) {module=ScaleMeshSize001}
   &ScaleMeshSize(Scale Factor=&VectorFromFloat(Value=0.08))
   &UpdateMeshOrientation(Rotation Rate=0.2)
   &ScaleColor(
    Scale Alpha=&FloatFromCurve(
     FloatCurve=DataInterface() {propertyValues {Curve {keys [
         {interpMode=RCIM_Cubic, tangentMode=RCTM_User, time=0, value=1}
         {arriveTangent=-0.8306, interpMode=RCIM_Cubic, leaveTangent=-0.8306, tangentMode=RCTM_User, time=0.7, value=0.8}
         {arriveTangent=-3.7878, interpMode=RCIM_Cubic, leaveTangent=-3.7878, tangentMode=RCTM_User, time=1, value=0}
        ]}}}
     CurveIndex=$System.NormalizedLoopAge
    )
    ScaleA=true
    ScaleRGB=false
   )
  ]
  rendererClasses [&NiagaraMeshRendererProperties]
  renderers [{rendererClass=&NiagaraMeshRendererProperties, rendererIndex=0}]
 }
}
summary {
 systemName=NS_SupplyDrop_Entry
 emitters [
  {emitterName=MainAirFlash, rendererClasses [&NiagaraMeshRendererProperties]}
  {emitterName=Particle_Emitter, rendererClasses [&NiagaraMeshRendererProperties]}
  {emitterName=Central_Flames, rendererClasses [&NiagaraMeshRendererProperties]}
 ]
 userVariables [
  User.Scale {type=NiagaraFloat, defaultValue=1, description=""}
  User.Time {type=NiagaraFloat, defaultValue=2, description=""}
 ]
}
